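# Sound cue for the final impact of a safe being demolished, playing randomized safe explosion waves on an object-impact sound class.
# A Modulator feeds a Quality Level switch, whose first two slots share a three-wave Random node and last two a two-wave one.
# The two-wave Random is excluded from branch culling.
# Medium-object destruction attenuation shapes falloff, and a four-max impact concurrency set is assigned alongside a PreventNew override.
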SoundCue /Game/Sounds/Foley_Loot/Containers/Safe/Demo_Safe_ImpactFinal_Cue
root=Modulator_6
nodes [
 {id=Modulator_6, children [QualityLevel_1]}
 {id=QualityLevel_1, children [Random_6,Random_6,Random_8,Random_8]}
 {id=Random_6, children [WavePlayer_13, WavePlayer_14, WavePlayer_15], properties {weights [1,1,1]}}
 {id=Random_8, children [WavePlayer_19, WavePlayer_18], properties {bShouldExcludeFromBranchCulling=true, weights [1,1]}}
 {id=WavePlayer_13, properties {soundWaveAssetPtr=&/Cash_System_Safe_Explo_03}}
 {id=WavePlayer_14, properties {soundWaveAssetPtr=&/Cash_System_Safe_Explo_01}}
 {id=WavePlayer_15, properties {soundWaveAssetPtr=&/Cash_System_Safe_Explo_02}}
 {id=WavePlayer_18, properties {soundWaveAssetPtr=&Cash_System_Safe_Explo_02}}
 {id=WavePlayer_19, properties {soundWaveAssetPtr=&Cash_System_Safe_Explo_01}}
]
properties {
 attenuationSettings=/Game/Sounds/Attenuation/Fort_Demolition_Destroy_Medium_Sized_Object_Att
 maxDistance=4000
 soundClassObject=&Fort_Impacts_PSM
 volumeMultiplier=0.8
 concurrencyOverrides {resolutionRule=PreventNew}
 concurrencySet [&CONC_Object_Impacts_FourMax]
}
