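# Sound cue for demolition impacts on a safe, playing a randomly chosen safe impact wave through small-object hit attenuation and the impacts sound class.
# A quality-level root sends its first two levels to an equally weighted random pick of three waves, and its last two to a pick of two.
# The two-wave branch is excluded from branch culling.
# An object-impact concurrency set is assigned, alongside override values that stop the oldest sound.
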SoundCue /Game/Sounds/Foley_Loot/Containers/Safe/Demo_Safe_Impact_Cue
root=QualityLevel_0
nodes [
 {id=QualityLevel_0, children [Random_3,Random_3,Random_4,Random_4]}
 {id=Random_3, children [WavePlayer_3, WavePlayer_7, WavePlayer_11], properties {weights [1,1,1]}}
 {id=Random_4, children [WavePlayer_12, WavePlayer_13], properties {bShouldExcludeFromBranchCulling=true, weights [1,1]}}
 {id=WavePlayer_11, properties {soundWaveAssetPtr=&/Cash_System_Safe_Impact_03}}
 {id=WavePlayer_12, properties {soundWaveAssetPtr=&/Cash_System_Safe_Impact_02}}
 {id=WavePlayer_13, properties {soundWaveAssetPtr=&/Cash_System_Safe_Impact_01}}
 {id=WavePlayer_3, properties {soundWaveAssetPtr=&Cash_System_Safe_Impact_01}}
 {id=WavePlayer_7, properties {soundWaveAssetPtr=&Cash_System_Safe_Impact_02}}
]
properties {
 attenuationSettings=/Game/Sounds/Attenuation/Fort_Demolition_Hit_Small_Sized_Object_Att
 maxDistance=4000
 soundClassObject=&Fort_Impacts
 volumeMultiplier=0.8
 concurrencyOverrides {maxCount=3, resolutionRule=StopOldest}
 concurrencySet [&CONC_Object_Impacts_FourMax]
}
