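# Sound cue for opening a tiered safe loot container, layering the open sound with a reward fanfare.
# A Mixer plays both waves together, with the fanfare at reduced level.
# Only the safe-open wave passes through a Modulator, which randomizes its pitch upward and lowers its volume to a fixed level.
# The cue uses item-interact attenuation, the loot reward sound class, and a concurrency override capping simultaneous instances at two.
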SoundCue /Game/Sounds/Foley_Loot/Containers/Safe/Tiered_Safe_Open_Cue
root=Mixer_0
nodes [
 {id=Mixer_0, children [Modulator_3,WavePlayer_1], properties {inputVolume [1,0.65]}}
 {id=Modulator_3, children [WavePlayer_0], properties {pitchMax=1.5, pitchMin=1, volumeMax=0.7, volumeMin=0.7}}
 {id=WavePlayer_0, properties {soundWaveAssetPtr=&/Cash_System_Safe_Open}}
 {id=WavePlayer_1, properties {soundWaveAssetPtr=&/Cash_System_Safe_Open_Fanfare_01}}
]
properties {
 attenuationSettings=/Game/Sounds/Attenuation/Fort_Item_Interact_Att
 bOverrideConcurrency=true
 maxDistance=1850
 soundClassObject=&Fort_Loot_Reward_PSM
 volumeMultiplier=1.2
 concurrencyOverrides {maxCount=2}
}
