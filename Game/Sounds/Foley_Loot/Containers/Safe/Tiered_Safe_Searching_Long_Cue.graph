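# Sound Cue for the long searching foley of a tiered safe, routed to the loot-collect sound class with item-interact mono attenuation.
# A Mixer layers a looping safe-search wave with a looped treasure chest search wave, each through its own pitch and volume Modulator.
# An Enveloper fades the chest layer out over about two seconds, leaving the safe loop.
# Concurrency allows one instance per owner and blocks new ones.
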SoundCue /Game/Sounds/Foley_Loot/Containers/Safe/Tiered_Safe_Searching_Long_Cue
root=Mixer_0
nodes [
 {
  id=Enveloper_0
  children [Looping_0]
  properties {
   pitchCurve {editorCurveData {keys [{time=0}]}}
   volumeCurve {editorCurveData {keys [{arriveTangent=-1.3539, interpMode=RCIM_Cubic, leaveTangent=-1.3539, tangentMode=RCTM_User, time=0, value=1}, <at 2.0174 -0.0448 RCIM_Cubic>]}}
  }
 }
 {id=Looping_0, children [WavePlayer_1]}
 {id=Mixer_0, children [Modulator_0,Modulator_2], properties {inputVolume [1,1]}}
 {id=Modulator_0, children [WavePlayer_0]}
 {id=Modulator_2, children [Enveloper_0], properties {pitchMax=1.07, pitchMin=0.92, volumeMax=1.07, volumeMin=0.92}}
 {id=WavePlayer_0, properties {bLooping=true, soundWaveAssetPtr=&/Safe_Search_Loop}}
 {id=WavePlayer_1, properties {soundWaveAssetPtr=/Game/Sounds/Foley_Loot/Containers/Treasure_Chest/treasure_search}}
]
properties {
 attenuationSettings=/Game/Sounds/Attenuation/Fort_Item_Interact_Mono_Att
 bOverrideConcurrency=true
 maxDistance=1500
 soundClassObject=&Fort_Loot_Collect
 volumeMultiplier=1.4125
 concurrencyOverrides {bLimitToOwner=true, maxCount=1, resolutionRule=PreventNew}
}
